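# Material function that generates an animated CRT-style RGB phosphor dot pattern from texture coordinates.
# Centered UVs are tiled, rotated by Angle, stretched 3:1, then spun and panned over time, though PanSpeedX lacks a Time connection, unlike PanSpeedY.
# Dot products on fractional UVs form round dots, shifted by thirds per channel.
# A second, half-cell staggered grid merges via Max, and Power sharpens the result.
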MaterialFunction /Game/Global/MaterialFunctions/Cosmetics/MF_CRT_RGB_Dots
graph {nodes [
  Add()
  n1 = ConstantBiasScale(None=FunctionInput(Preview=TextureCoordinate()) {bUsePreviewValueAsDefault=true, inputName=Texture Coordinates, inputType=FunctionInput_Vector2}) {bias=-0.5, scale=2}
  n2 = FunctionInput(Preview=Constant() {r=10}) {bUsePreviewValueAsDefault=true, inputName=Tiling, inputType=FunctionInput_Scalar, sortPriority=2}
  n3 = FunctionInput(Preview=Constant()) {bUsePreviewValueAsDefault=true, inputName=Angle, inputType=FunctionInput_Scalar, sortPriority=3}
  n4 = Multiply(
   A=Constant2Vector() {g=1, r=3}
   B=Rotator(Coordinate=Multiply(A=@n1, B=@n2), Time=Divide(A=@n3) {constB=57.2958}) {centerX=0, centerY=0, speed=1}
  )
  n5 = Time() {bOverride_Period=true, period=600}
  n6 = Multiply(
   A=@n5
   B=FunctionInput(Preview=Constant()) {bUsePreviewValueAsDefault=true, inputName=RotateSpeed, inputType=FunctionInput_Scalar, sortPriority=8}
  )
  n7 = FunctionInput(Preview=Constant()) {bUsePreviewValueAsDefault=true, inputName=PanSpeedY, inputType=FunctionInput_Scalar, sortPriority=7}
  n8 = Constant()
  n9 = Panner(
   Coordinate=Panner(Coordinate=Rotator(Coordinate=@n4, Time=@n6) {centerX=0, centerY=0, speed=1}, Time=Multiply(A=@n5, B=@n7)) {speedY=1}
   Time=Multiply(B=FunctionInput(Preview=@n8) {bUsePreviewValueAsDefault=true, inputName=PanSpeedX, inputType=FunctionInput_Scalar, sortPriority=6})
  ) {speedX=1}
  n10 = ConstantBiasScale(None=Subtract(A=@n9, B=Floor(None=@n9))) {bias=-0.5, scale=2}
  n11 = OneMinus(None=DotProduct(A=@n10, B=@n10))
  n12 = Constant2Vector() {r=0.3333}
  n13 = Add(A=@n9, B=@n12) {constB=0.3333}
  n14 = ConstantBiasScale(None=Subtract(A=@n13, B=Floor(None=@n13))) {bias=-0.5, scale=2}
  n15 = OneMinus(None=DotProduct(A=@n14, B=@n14))
  n16 = Add(A=@n13, B=@n12) {constB=0.3333}
  n17 = ConstantBiasScale(None=Subtract(A=@n16, B=Floor(None=@n16))) {bias=-0.5, scale=2}
  n18 = DotProduct(A=@n17, B=@n17)
  n19 = AppendVector(A=AppendVector(A=Saturate(None=@n11), B=Saturate(None=@n15)), B=Saturate(None=OneMinus(None=@n18)))
  n20 = FunctionInput(Preview=Constant() {r=1}) {bUsePreviewValueAsDefault=true, inputName=Offset, inputType=FunctionInput_Scalar, sortPriority=1}
  n21 = Constant() {r=0.5}
  n22 = Add(A=@n9, B=AppendVector(A=Multiply(A=@n20, B=@n21), B=@n21))
  n23 = ConstantBiasScale(None=Subtract(A=@n22, B=Floor(None=@n22))) {bias=-0.5, scale=2}
  n24 = OneMinus(None=DotProduct(A=@n23, B=@n23))
  n25 = Constant2Vector() {r=0.3333}
  n26 = Add(A=@n22, B=@n25) {constB=0.3333}
  n27 = ConstantBiasScale(None=Subtract(A=@n26, B=Floor(None=@n26))) {bias=-0.5, scale=2}
  n28 = OneMinus(None=DotProduct(A=@n27, B=@n27))
  n29 = Add(A=@n26, B=@n25) {constB=0.3333}
  n30 = ConstantBiasScale(None=Subtract(A=@n29, B=Floor(None=@n29))) {bias=-0.5, scale=2}
  n31 = DotProduct(A=@n30, B=@n30)
  n32 = AppendVector(A=AppendVector(A=Saturate(None=@n24), B=Saturate(None=@n28)), B=Saturate(None=OneMinus(None=@n31)))
  n33 = Constant() {r=6}
  FunctionOutput(None=Power(
    Base=Max(A=@n19, B=@n32)
    Exp=FunctionInput(Preview=@n33) {bUsePreviewValueAsDefault=true, inputName=Sharpen, inputType=FunctionInput_Scalar, sortPriority=5}
   ))
 ]}
