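# Material function that generates 0 to 1 UVs purely from pixel normal direction relative to the camera, with polar and somewhat distorted Cartesian outputs.
# Radius comes from the arcsine of normal dot view vector.
# Angle comes from camera-space XY of normal minus view, using sign tests to span the full circle.
# Sine and cosine convert polar to Cartesian; quality switches use fast arcsine on Low and Medium.
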MaterialFunction /Game/Global/MaterialFunctions/Cosmetics/MF_UVsFromSphericalNormals
graph {
 properties {
  description="Generates 0...1 UV coordinates based only on surface normal direction. Radial and (somewhat distorted) Cartesian outputs provided"
  libraryCategoriesText [Coordinates]
 }
 nodes [
  MaterialExpressionNamedRerouteUsage_0 = NamedRerouteUsage()
  n1 = Convert(None=@MaterialExpressionNamedRerouteUsage_0) {
   nodeName=Break Float2
   convertInputs [{
     type=Vector2
     expressionInput {
      expression=&self:MaterialExpressionNamedRerouteUsage_0
      inputName=None
      mask=0
      maskA=0
      maskB=0
      maskG=0
      maskR=0
      outputIndex=0
     }
    }]
   convertOutputs [{type=Scalar}, {type=Scalar}]
  }
  n2 = Cosine(None=@n1)
  n3 = Sine(None=@n1)
  n4 = Add(A=Multiply(A=@n1, B=AppendVector(A=@n2, B=@n3))) {constB=-1}
  FunctionOutput(None=NamedRerouteDeclaration(None=Multiply(A=@n4) {constB=-0.5}) {name=PolarToCartesian, nodeColor=<rgba 1 0 0.5156 1>}) {outputName=CartesianCoords, sortPriority=10}
  n5 = DotProduct(A=PixelNormalWS(), B=CameraVectorWS())
  n6 = ArcsineFast!(None=@n5)
  n7 = Divide(
   A=QualitySwitch(Default=Arcsine!(None=@n5), Low=@n6, Medium=@n6)
   B=MaterialFunctionCall() {materialFunction=&Pi}.Result
  )
  n8 = Subtract(A=PixelNormalWS(), B=CameraVectorWS())
  n9 = Normalize(VectorInput=Multiply(
    A=Transform(None=@n8) {transformSourceType=TRANSFORMSOURCE_World, transformType=TRANSFORM_Camera}
    B=Constant3Vector() {constant=<rgba 1 1 0 0>}
   ))
  n10 = Constant3Vector() {constant=<rgba 0 1 0 0>}
  n11 = ComponentMask(None=CrossProduct(A=@n9, B=@n10)) {b=true}
  n12 = Abs(None=@n11)
  n13 = Arcsine(None=@n12)
  n14 = ArcsineFast(None=@n12)
  n15 = Constant() {r=0.2}
  n16 = Multiply(
   A=Multiply(
    A=QualitySwitch(Default=@n13, Low=@n14, Medium=@n14)
    B=MaterialFunctionCall(Multiplier=@n15) {materialFunction=&Pi}.Result
   )
   B=Sign(None=@n11)
  )
  n17 = DotProduct(A=@n9, B=@n10)
  n18 = LinearInterpolate(A=ConstantBiasScale(None=@n16) {bias=-3, scale=-0.25}, B=ConstantBiasScale(None=@n16) {scale=0.25}, Alpha=Saturate(None=Sign(None=@n17))) {constB=0}
  n19 = Subtract(A=NamedRerouteDeclaration(None=AppendVector(A=@n7, B=@n18)) {name=RadialCoords, nodeColor=<rgba 0.7578 0 1 1>}, B=Constant2Vector() {r=0.5})
  n20 = Constant2Vector() {g=1, r=-2}
  FunctionOutput(None=NamedRerouteDeclaration(None=Multiply(A=@n19, B=@n20)) {name=PolarCoordinates, nodeColor=<rgba 1 0 0.3516 1>}) {outputName=PolarCoords}
 ]
}
