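# Material function that draws a tiled grid whose cells pulse between two colors under colored lines, described as base colors for the death transition.
# Ceiled tiled UVs seed vector noise, giving each cell a random speed and phase for a time-driven sine blend.
# Sphere masks on mirrored per-tile UVs form adjustable edge lines.
# Besides color, it outputs grayscale, random and line masks, plus stepped cell-center and per-tile UVs.
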MaterialFunction /Game/Global/MaterialFunctions/Cosmetics/AnimatedGrid/MF_AnimatedGrid
graph {
 expressions {
  MaterialExpressionAbs_0 {class=MaterialExpressionAbs, inputs [None {from=MaterialExpressionConstantBiasScale_1, output=""}], properties {}}
  MaterialExpressionAdd_0 {
   class=MaterialExpressionAdd
   inputs [
    A {from=MaterialExpressionSphereMask_0}
    B {from=MaterialExpressionSphereMask_1}
   ]
  }
  MaterialExpressionAdd_5 {
   class=MaterialExpressionAdd
   inputs [
    A {from=MaterialExpressionMultiply_8}
    B {from=""}
   ]
   properties {constA=1}
  }
  MaterialExpressionAdd_6 {
   class=MaterialExpressionAdd
   inputs [
    A {from=MaterialExpressionMultiply_10}
    B {from=MaterialExpressionComponentMask_4}
   ]
  }
  MaterialExpressionAppendVector_0 {
   class=MaterialExpressionAppendVector
   inputs [
    A {from=MaterialExpressionNamedRerouteDeclaration_8, output=""}
    B {from=MaterialExpressionConstant_6, output=""}
   ]
  }
  MaterialExpressionCeil_0 {class=MaterialExpressionCeil, inputs [None {from=MaterialExpressionNamedRerouteUsage_0, output=""}], properties {}}
  MaterialExpressionComponentMask_1 {class=MaterialExpressionComponentMask, inputs [None {from=MaterialExpressionAbs_0}], properties {r=true}}
  MaterialExpressionComponentMask_2 {class=MaterialExpressionComponentMask, inputs [None {from=MaterialExpressionAbs_0}], properties {g=true}}
  MaterialExpressionComponentMask_3 {class=MaterialExpressionComponentMask, inputs [None {from=MaterialExpressionNamedRerouteUsage_2}], properties {r=true}}
  MaterialExpressionComponentMask_4 {class=MaterialExpressionComponentMask, inputs [None {from=MaterialExpressionNamedRerouteUsage_3}], properties {g=true}}
  MaterialExpressionConstantBiasScale_1 {class=MaterialExpressionConstantBiasScale, inputs [None {from=MaterialExpressionFunctionOutput_5}], properties {bias=-0.5, scale=2}}
  MaterialExpressionConstantBiasScale_2 {class=MaterialExpressionConstantBiasScale, inputs [None {from=MaterialExpressionSine_1}]}
  MaterialExpressionConstant_6 {class=MaterialExpressionConstant}
  MaterialExpressionConstant_7 {class=MaterialExpressionConstant, properties {r=1}}
  MaterialExpressionConvert_0 {
   class=MaterialExpressionConvert
   inputs [
    None {from=MaterialExpressionFunctionInput_3}
    None {from=MaterialExpressionFunctionInput_5}
   ]
   properties {
    nodeName=Make Vector2
    convertInputs [
     {
      type=Scalar
      expressionInput {
       expression=&self:MaterialExpressionFunctionInput_3
       inputName=None
       mask=0
       maskA=0
       maskB=0
       maskG=0
       maskR=0
       outputIndex=0
      }
     }
     {
      type=Scalar
      expressionInput {
       expression=&self:MaterialExpressionFunctionInput_5
       inputName=None
       mask=0
       maskA=0
       maskB=0
       maskG=0
       maskR=0
       outputIndex=0
      }
     }
    ]
    convertOutputs [{type=Vector2}]
   }
  }
  MaterialExpressionDivide_0 {
   class=MaterialExpressionDivide
   inputs [
    A {from=MaterialExpressionSubtract_1}
    B {from=MaterialExpressionNamedRerouteUsage_10}
   ]
  }
  MaterialExpressionFrac_0 {class=MaterialExpressionFrac, inputs [None {from=MaterialExpressionNamedRerouteUsage_1, output=""}], properties {}}
  MaterialExpressionFunctionInput_1 {
   class=MaterialExpressionFunctionInput
   inputs [Preview {from=MaterialExpressionTextureCoordinate_4, output=""}]
   properties {bUsePreviewValueAsDefault=true, inputName=UV, inputType=FunctionInput_Vector2}
  }
  MaterialExpressionFunctionInput_10 {
   class=MaterialExpressionFunctionInput
   inputs [Preview {from="", output=""}]
   properties {bUsePreviewValueAsDefault=true, inputName=Color 01, sortPriority=10, previewValue=<0.05 0.25 0.35 1>}
  }
  MaterialExpressionFunctionInput_11 {
   class=MaterialExpressionFunctionInput
   inputs [Preview {from="", output=""}]
   properties {bUsePreviewValueAsDefault=true, inputName=Color 02, sortPriority=11, previewValue=<0 0.65 1 1>}
  }
  MaterialExpressionFunctionInput_12 {
   class=MaterialExpressionFunctionInput
   inputs [Preview {from="", output=""}]
   properties {bUsePreviewValueAsDefault=true, inputName=Line Color, sortPriority=32, previewValue=<0 3 6 1>}
  }
  MaterialExpressionFunctionInput_2 {
   class=MaterialExpressionFunctionInput
   inputs [Preview {from="", output=""}]
   properties {bUsePreviewValueAsDefault=true, inputName=Tiling, inputType=FunctionInput_Scalar, sortPriority=1, previewValue=<20 0 0 1>}
  }
  MaterialExpressionFunctionInput_3 {
   class=MaterialExpressionFunctionInput
   inputs [Preview {from="", output=""}]
   properties {bUsePreviewValueAsDefault=true, inputName=Tiling X, inputType=FunctionInput_Scalar, sortPriority=2, previewValue=<1 0 0 1>}
  }
  MaterialExpressionFunctionInput_5 {
   class=MaterialExpressionFunctionInput
   inputs [Preview {from="", output=""}]
   properties {bUsePreviewValueAsDefault=true, inputName=Tiling Y, inputType=FunctionInput_Scalar, sortPriority=3, previewValue=<1 0 0 1>}
  }
  MaterialExpressionFunctionInput_6 {
   class=MaterialExpressionFunctionInput
   inputs [Preview {from="", output=""}]
   properties {bUsePreviewValueAsDefault=true, inputName=Line Thickness, inputType=FunctionInput_Scalar, sortPriority=30, previewValue=<0.1 0 0 1>}
  }
  MaterialExpressionFunctionInput_7 {
   class=MaterialExpressionFunctionInput
   inputs [Preview {from="", output=""}]
   properties {bUsePreviewValueAsDefault=true, inputName=Line Hardness, inputType=FunctionInput_Scalar, sortPriority=31, previewValue=<1 0 0 1>}
  }
  MaterialExpressionFunctionInput_8 {
   class=MaterialExpressionFunctionInput
   inputs [Preview {from="", output=""}]
   properties {bUsePreviewValueAsDefault=true, inputName=Speed Random Amount, inputType=FunctionInput_Scalar, sortPriority=21, previewValue=<2 0 0 1>}
  }
  MaterialExpressionFunctionInput_9 {
   class=MaterialExpressionFunctionInput
   inputs [Preview {from="", output=""}]
   properties {bUsePreviewValueAsDefault=true, inputName=Speed, inputType=FunctionInput_Scalar, sortPriority=20, previewValue=<0.2 0 0 1>}
  }
  MaterialExpressionFunctionOutput_0 {
   class=MaterialExpressionFunctionOutput
   inputs [None {from=MaterialExpressionLinearInterpolate_4, output=""}]
   properties {outputName="Result (V3)"}
  }
  MaterialExpressionFunctionOutput_1 {
   class=MaterialExpressionFunctionOutput
   inputs [None {from=MaterialExpressionNamedRerouteDeclaration_9, output=""}]
   properties {outputName=Stepped UV, sortPriority=4}
  }
  MaterialExpressionFunctionOutput_2 {
   class=MaterialExpressionFunctionOutput
   inputs [None {from=MaterialExpressionNamedRerouteDeclaration_2, output=""}]
   properties {outputName=Gird Lines, sortPriority=3}
  }
  MaterialExpressionFunctionOutput_3 {
   class=MaterialExpressionFunctionOutput
   inputs [None {from=MaterialExpressionNamedRerouteDeclaration_1, output=""}]
   properties {outputName=Random Grid, sortPriority=2}
  }
  MaterialExpressionFunctionOutput_4 {
   class=MaterialExpressionFunctionOutput
   inputs [None {from=MaterialExpressionNamedRerouteDeclaration_3, output=""}]
   properties {outputName=Grayscale Grid, sortPriority=1}
  }
  MaterialExpressionFunctionOutput_5 {
   class=MaterialExpressionFunctionOutput
   inputs [None {from=MaterialExpressionFrac_0, output=""}]
   properties {outputName=Per Tile UV, sortPriority=5}
  }
  MaterialExpressionLinearInterpolate_3 {
   class=MaterialExpressionLinearInterpolate
   inputs [
    A {from=MaterialExpressionFunctionInput_10}
    B {from=MaterialExpressionFunctionInput_11}
    Alpha {from=MaterialExpressionNamedRerouteUsage_4}
   ]
  }
  MaterialExpressionLinearInterpolate_4 {
   class=MaterialExpressionLinearInterpolate
   inputs [
    A {from=MaterialExpressionLinearInterpolate_3}
    B {from=MaterialExpressionFunctionInput_12}
    Alpha {from=MaterialExpressionNamedRerouteUsage_5}
   ]
  }
  MaterialExpressionMultiply_10 {
   class=MaterialExpressionMultiply
   inputs [
    A {from=MaterialExpressionTime_2}
    B {from=MaterialExpressionMultiply_9}
   ]
  }
  MaterialExpressionMultiply_4 {
   class=MaterialExpressionMultiply
   inputs [
    A {from=MaterialExpressionFunctionInput_1}
    B {from=MaterialExpressionNamedRerouteDeclaration_6}
   ]
  }
  MaterialExpressionMultiply_5 {
   class=MaterialExpressionMultiply
   inputs [
    A {from=MaterialExpressionFunctionInput_2}
    B {from=MaterialExpressionConvert_0}
   ]
  }
  MaterialExpressionMultiply_8 {
   class=MaterialExpressionMultiply
   inputs [
    A {from=MaterialExpressionComponentMask_3}
    B {from=MaterialExpressionFunctionInput_8}
   ]
  }
  MaterialExpressionMultiply_9 {
   class=MaterialExpressionMultiply
   inputs [
    A {from=MaterialExpressionAdd_5}
    B {from=MaterialExpressionFunctionInput_9}
   ]
  }
  MaterialExpressionNamedRerouteDeclaration_0 {
   class=MaterialExpressionNamedRerouteDeclaration
   inputs [None {from=MaterialExpressionMultiply_4}]
   properties {name=UVs, nodeColor=<rgba 0 0.0391 1 1>}
  }
  MaterialExpressionNamedRerouteDeclaration_1 {
   class=MaterialExpressionNamedRerouteDeclaration
   inputs [None {from=MaterialExpressionVectorNoise_0}]
   properties {name=RandomGrid, nodeColor=<rgba 0.3594 1 0 1>}
  }
  MaterialExpressionNamedRerouteDeclaration_2 {
   class=MaterialExpressionNamedRerouteDeclaration
   inputs [None {from=MaterialExpressionSaturate_0}]
   properties {name=GridLines, nodeColor=<rgba 1 0 0.6797 1>}
  }
  MaterialExpressionNamedRerouteDeclaration_3 {
   class=MaterialExpressionNamedRerouteDeclaration
   inputs [None {from=MaterialExpressionConstantBiasScale_2}]
   properties {name=AnimatedGrid, nodeColor=<rgba 0.4531 0 1 1>}
  }
  MaterialExpressionNamedRerouteDeclaration_4 {
   class=MaterialExpressionNamedRerouteDeclaration
   inputs [None {from=MaterialExpressionFunctionInput_6}]
   properties {name=LineThickness, nodeColor=<rgba 0 1 0.1328 1>}
  }
  MaterialExpressionNamedRerouteDeclaration_5 {
   class=MaterialExpressionNamedRerouteDeclaration
   inputs [None {from=MaterialExpressionFunctionInput_7}]
   properties {name=LineHardness, nodeColor=<rgba 1 0 0.1875 1>}
  }
  MaterialExpressionNamedRerouteDeclaration_6 {
   class=MaterialExpressionNamedRerouteDeclaration
   inputs [None {from=MaterialExpressionMultiply_5}]
   properties {name=UVTiling, nodeColor=<rgba 0.1016 1 0 1>}
  }
  MaterialExpressionNamedRerouteDeclaration_8 {
   class=MaterialExpressionNamedRerouteDeclaration
   inputs [None {from=MaterialExpressionCeil_0}]
   properties {name=CeiledUVs, nodeColor=<rgba 0 0.7422 1 1>}
  }
  MaterialExpressionNamedRerouteDeclaration_9 {
   class=MaterialExpressionNamedRerouteDeclaration
   inputs [None {from=MaterialExpressionDivide_0}]
   properties {name=SteppedUVs, nodeColor=<rgba 1 0.4453 0 1>}
  }
  MaterialExpressionNamedRerouteUsage_0 {class=MaterialExpressionNamedRerouteUsage}
  MaterialExpressionNamedRerouteUsage_1 {class=MaterialExpressionNamedRerouteUsage}
  MaterialExpressionNamedRerouteUsage_10 {class=MaterialExpressionNamedRerouteUsage}
  MaterialExpressionNamedRerouteUsage_2 {class=MaterialExpressionNamedRerouteUsage}
  MaterialExpressionNamedRerouteUsage_3 {class=MaterialExpressionNamedRerouteUsage}
  MaterialExpressionNamedRerouteUsage_4 {class=MaterialExpressionNamedRerouteUsage}
  MaterialExpressionNamedRerouteUsage_5 {class=MaterialExpressionNamedRerouteUsage}
  MaterialExpressionNamedRerouteUsage_6 {class=MaterialExpressionNamedRerouteUsage}
  MaterialExpressionNamedRerouteUsage_7 {class=MaterialExpressionNamedRerouteUsage}
  MaterialExpressionNamedRerouteUsage_9 {class=MaterialExpressionNamedRerouteUsage}
  MaterialExpressionSaturate_0 {class=MaterialExpressionSaturate, inputs [None {from=MaterialExpressionAdd_0}]}
  MaterialExpressionSine_1 {class=MaterialExpressionSine, inputs [None {from=MaterialExpressionAdd_6}]}
  MaterialExpressionSphereMask_0 {
   class=MaterialExpressionSphereMask
   inputs [
    A {from=MaterialExpressionComponentMask_1, output=""}
    B {from=MaterialExpressionConstant_7, output=""}
    Radius {from=MaterialExpressionNamedRerouteUsage_6, output=""}
    Hardness {from=MaterialExpressionNamedRerouteUsage_7, output=""}
   ]
   properties {}
  }
  MaterialExpressionSphereMask_1 {
   class=MaterialExpressionSphereMask
   inputs [
    A {from=MaterialExpressionConstant_7, output=""}
    B {from=MaterialExpressionComponentMask_2, output=""}
    Radius {from=MaterialExpressionNamedRerouteUsage_6, output=""}
    Hardness {from=MaterialExpressionNamedRerouteUsage_7, output=""}
   ]
   properties {}
  }
  MaterialExpressionSubtract_1 {
   class=MaterialExpressionSubtract
   inputs [
    A {from=MaterialExpressionNamedRerouteUsage_9}
    B {from=""}
   ]
   properties {constB=0.5}
  }
  MaterialExpressionTextureCoordinate_4 {class=MaterialExpressionTextureCoordinate}
  MaterialExpressionTime_2 {class=MaterialExpressionTime}
  MaterialExpressionVectorNoise_0 {class=MaterialExpressionVectorNoise, inputs [World Position {from=MaterialExpressionAppendVector_0, output=""}], properties {}}
 }
 properties {description=Base Colors for the death Transition}
}
