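# Material function supplying animated grid base colors for the death transition cosmetic effect.
# A time-driven sine, with speed and offset apparently varied per cell by the grayscale Grid input, is remapped into a pulsing grayscale value.
# That value blends Color 01 and Color 02, then the Grid Lines mask overlays Line Color.
# It outputs both the colored result and the animated grayscale grid.
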MaterialFunction /Game/Global/MaterialFunctions/Cosmetics/AnimatedGrid/MF_AnimatedGridTexture
graph {
 properties {description=Base Colors for the death Transition}
 nodes [
  n1 = FunctionInput() {bUsePreviewValueAsDefault=true, inputName=Color 01, sortPriority=10, previewValue=<0.05 0.25 0.35 1>}
  n2 = FunctionInput() {bUsePreviewValueAsDefault=true, inputName=Color 02, sortPriority=11, previewValue=<0 0.65 1 1>}
  n3 = NamedRerouteUsage()
  n4 = FunctionInput() {bUsePreviewValueAsDefault=true, inputName=Line Color, sortPriority=32, previewValue=<0 3 6 1>}
  FunctionOutput(None=LinearInterpolate(A=LinearInterpolate(A=@n1, B=@n2, Alpha=@n3), B=@n4, Alpha=NamedRerouteUsage())) {outputName="Result (V3)"}
  n5 = Time()
  n6 = NamedRerouteUsage()
  n7 = FunctionInput() {bUsePreviewValueAsDefault=true, inputName=Speed Random Amount, inputType=FunctionInput_Scalar, sortPriority=21, previewValue=<2 0 0 1>}
  n8 = FunctionInput() {bUsePreviewValueAsDefault=true, inputName=Speed, inputType=FunctionInput_Scalar, sortPriority=20, previewValue=<0.2 0 0 1>}
  n9 = Multiply(A=Add(A=Multiply(A=@n6, B=@n7)) {constA=1}, B=@n8)
  n10 = Sine(None=Add(A=Multiply(A=@n5, B=@n9), B=NamedRerouteUsage()))
  FunctionOutput(None=NamedRerouteDeclaration(None=ConstantBiasScale(None=@n10)) {name=AnimatedGrid, nodeColor=<rgba 0.4531 0 1 1>}) {outputName=Grayscale Grid, sortPriority=1}
  n11 = FunctionInput() {bUsePreviewValueAsDefault=true, inputName=Grid, inputType=FunctionInput_Scalar, previewValue=<1 0 0 1>}
  NamedRerouteDeclaration(None=@n11) {name=GrayscaleGrid, nodeColor=<rgba 0 0.0391 1 1>}
  NamedRerouteDeclaration(None=FunctionInput() {bUsePreviewValueAsDefault=true, inputName=Grid Lines, inputType=FunctionInput_Scalar, sortPriority=1}) {name=GridLines, nodeColor=<rgba 1 0 0.6797 1>}
 ]
}
